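# Translucent lit cinematic smoke material for Cascade and Niagara sprites, using volumetric per-vertex directional lighting.
# The SmokeLerp dynamic parameter blends two 8x8 smoke flipbooks for normals, color mask and opacity.
# An 8x-tiled spherical normal, softened by FlattenNormal, is angle-corrected into the flipbook normal to round each frame.
# Particle color tints a mask-driven blend of two color parameters, while opacity uses a texture-modulated DepthFade distance and particle alpha.
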Material /Game/Effects/Fort_Effects/Materials/Smoke/Cinematic/M_Lit_Cinematic_Smoke_SubUV
graph {
 properties {
  bUsedWithNiagaraSprites=true
  bUsedWithParticleSprites=true
  blendMode=BLEND_Translucent
  pixelDepthOffsetMode=PDOM_Legacy
  translucencyDirectionalLightingIntensity=2
  translucencyLightingMode=TLM_VolumetricPerVertexDirectional
  usedShadingModels=Default Lit
 }
 nodes [
  n1 = TextureSample!(UVs, Tex, Level, Apply View MipBias) {
   constMipValue=3
   mipValueMode=TMVM_MipLevel
   samplerType=SAMPLERTYPE_Normal
   texture=/Game/Effects/Fort_Effects/Gameplay/SmokeDustExplosions/Textures/T_SmokeDeath_8x8_01_Normals
  }
  n2 = TextureSample() {
   constCoordinate=5
   mipValueMode=TMVM_MipLevel
   samplerType=SAMPLERTYPE_Normal
   texture=/Game/Effects/Fort_Effects/Textures/Smoke/T_Smoke_8x8_01_N
  }
  n3 = DynamicParameter() {defaultValue=<rgba 1 1 0.1 1>, paramNames [DepthFade, SmokeLerp, FlattenNormal, Param4]}
  n4 = TextureCoordinate() {uTiling=8, vTiling=8}
  MaterialFunctionCall(
   BaseNormal=LinearInterpolate(A=@n1.RGB, B=@n2.RGB, Alpha=@n3.SmokeLerp)
   AdditionalNormal=MaterialFunctionCall(
    Normal=TextureSample!(UVs=@n4, Tex, Apply View MipBias) {samplerType=SAMPLERTYPE_Normal, texture=/Game/Gadgets/Assets/Teleporter/Textures/T_Spherical_Normal_02}.RGB
    Flatness=DynamicParameter() {defaultValue=<rgba 1 1 0.1 1>, paramNames [DepthFade, SmokeLerp, FlattenNormal, Param4]}.FlattenNormal
   ) {materialFunction=&FlattenNormal}.Result
  ) {materialFunction=&BlendAngleCorrectedNormals}
  n5 = ParticleColor()
  Multiply(A=@n5.RGB) {constB=0.1}
  n6 = Constant()
  n7 = TextureSample!(UVs, Tex, Level=@n6, Apply View MipBias) {
   mipValueMode=TMVM_MipLevel
   samplerType=SAMPLERTYPE_LinearColor
   texture=/Game/Effects/Fort_Effects/Gameplay/SmokeDustExplosions/Textures/T_SmokeDeath_8x8_01
  }
  n8 = TextureSample(Level=@n6) {mipValueMode=TMVM_MipLevel, samplerType=SAMPLERTYPE_LinearColor, texture=/Game/Effects/Fort_Effects/Textures/Smoke/T_Smoke_8x8_01}
  n9 = Power(Base=@n8.G)
  n10 = DynamicParameter() {defaultValue=<rgba 1 1 0.1 1>, paramNames [DepthFade, SmokeLerp, FlattenNormal, Param4]}
  n11 = Multiply(A=@n10.DepthFade) {constB=0.5}
  Multiply(
   A=DepthFade(
    Opacity=LinearInterpolate(A=@n7.R, B=@n9, Alpha=@n3.SmokeLerp)
    FadeDistance=LinearInterpolate(A=@n10.DepthFade, B=@n11, Alpha=@n7.G) {constA=100, constB=200}
   )
   B=ParticleColor().A
  )
  Multiply(
   A=@n5.RGB
   B=LinearInterpolate(
    A=VectorParameter() {parameterName=BaseColor0, defaultValue=<rgba 1 1 1 0>}.RGB
    B=VectorParameter() {parameterName=BaseColor1, defaultValue=<rgba 0.9 0.9 0.9 0>}.RGB
    Alpha=LinearInterpolate(A=@n7.G, B=@n9, Alpha=@n3.SmokeLerp)
   )
  )
 ]
}
